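# Masked, two-sided subsurface material for pepperoni mesh particles, including Niagara, in the Pizza glider effects.
# It samples the PizzaPit glider diffuse, normal and mask textures, scaling diffuse and roughness by scalar parameters.
# The normal blends with a tiled Arid rock detail normal, then passes through FlattenNormal with adjustable flatness.
# Scalar and color parameters expose subsurface strength and a dark red tint.
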Material /Game/Effects/Fort_Effects/Materials/Pizza/M_PizzaGlider_Pepperoni
graph {
 properties {
  bUsedWithMeshParticles=true
  bUsedWithNiagaraMeshParticles=true
  blendMode=BLEND_Masked
  pixelDepthOffsetMode=PDOM_Legacy
  previewMesh=/Game/Effects/Fort_Effects/Mesh/Pizza/SM_GreenPepper_V1
  shadingModel=MSM_Subsurface
  twoSided=true
  usedShadingModels=Subsurface
 }
 nodes [
  Constant()
  n1 = TextureSampleParameter2D() {
   parameterName=Normal
   samplerType=SAMPLERTYPE_Normal
   texture=/Game/Gadgets/Assets/VinderTech_GliderChute/Glider_PizzaPit/Textures/T_Glider_PizzaPit_N
  }
  n2 = Multiply(A=TextureCoordinate(), B=ScalarParameter() {defaultValue=0.168, parameterName=Detail Tiling})
  MaterialFunctionCall(
   Normal=LinearInterpolate(A=@n1.RGB, B=TextureSample(UVs=@n2) {samplerType=SAMPLERTYPE_Normal, texture=/Game/Environments/Sets/Arid/Textures/RockDetail_normal}.RGB)
   Flatness=ScalarParameter() {defaultValue=0.5, parameterName=Flatten Normal}
  ) {materialFunction=&FlattenNormal}
  n3 = TextureSampleParameter2D() {
   parameterName=Roughness
   samplerType=SAMPLERTYPE_Masks
   texture=/Game/Gadgets/Assets/VinderTech_GliderChute/Glider_PizzaPit/Textures/T_Glider_PizzaPit_S
  }
  Multiply(A=@n3.RGB, B=ScalarParameter() {defaultValue=0.9945, parameterName=Roughness})
  Multiply(
   A=TextureSampleParameter2D() {parameterName=Diffuse, texture=/Game/Gadgets/Assets/VinderTech_GliderChute/Glider_PizzaPit/Textures/T_Glider_PizzaPit_D}.RGB
   B=ScalarParameter() {defaultValue=1, parameterName=Base Color Tint}
  )
  ScalarParameter() {defaultValue=1, parameterName=Subsurf}
  VectorParameter() {parameterName=SubSurf Color, defaultValue=<rgba 0.4271 0 0 0>}
 ]
}
